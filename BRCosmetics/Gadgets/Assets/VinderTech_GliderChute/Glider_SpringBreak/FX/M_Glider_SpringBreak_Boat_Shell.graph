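# Unlit translucent material for Niagara mesh particles, a water-like shell effect on the Spring Break boat glider.
# A base color plus panning foam swirls, tinted by particle color, feeds the emissive adjustments function.
# Opacity math multiplies two opposing V gradients, optionally blends toward vertex red, then applies a smoothstep-eroded side gradient centered on U.
# Vertex red, particle alpha and an opacity parameter scale that mask.
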Material /BRCosmetics/Gadgets/Assets/VinderTech_GliderChute/Glider_SpringBreak/FX/M_Glider_SpringBreak_Boat_Shell
graph {
 properties {
  bUsedWithNiagaraMeshParticles=true
  blendMode=BLEND_Translucent
  pixelDepthOffsetMode=PDOM_Legacy
  shadingModel=MSM_Unlit
  usedShadingModels=Unlit
 }
 nodes [
  n1 = VectorParameter() {parameterName=Color, defaultValue=<rgba 5 7.4167 10 0>}
  n2 = VectorParameter() {parameterName=Tiling, defaultValue=<rgba 1.333 0.5 0 1>}
  n3 = Panner(
   Coordinate=Multiply(A=TextureCoordinate(), B=ComponentMask(None=@n2.RGB) {g=true, r=true})
   Speed=VectorParameter() {parameterName=Speed, defaultValue=<rgba 0 0.8 0 1>}.RGB
  )
  n4 = Multiply(
   A=Multiply(
    A=VectorParameter() {parameterName=Color2, defaultValue=<rgba 8 8 8 0>}.RGB
    B=TextureSample(UVs=@n3) {samplerType=SAMPLERTYPE_LinearColor, texture=/Game/Global/Textures/Default/Caustics/03_WaterFoam_AccentSwirls_H}.R
   )
   B=ScalarParameter() {defaultValue=0.3, parameterName=FoamOpacity}
  )
  n5 = ParticleColor()
  MaterialFunctionCall(Emissive=Multiply(A=Add(A=@n1.RGB, B=@n4), B=@n5.RGB)) {materialFunction=&MF_FX_EmissiveAdjustments}
  n6 = ComponentMask(None=TextureCoordinate()) {g=true}
  n7 = ScalarParameter() {parameterName=Offset_Y}
  n8 = Multiply(
   A=Power(Base=Add(A=@n6, B=@n7), Exp=ScalarParameter() {defaultValue=1, parameterName=OpacityPower}) {constExponent=4}
   B=ScalarParameter() {defaultValue=15, parameterName=Param_2}
  )
  n9 = TextureCoordinate()
  n10 = Add(A=OneMinus(None=ComponentMask(None=@n9) {g=true}), B=ScalarParameter() {parameterName=Offset_Y_2})
  n11 = ScalarParameter() {defaultValue=1, parameterName=OpacityPower2}
  n12 = Multiply(
   A=Multiply(A=Saturate(None=@n8), B=Power(Base=@n10, Exp=@n11) {constExponent=4})
   B=ScalarParameter() {defaultValue=0.2, parameterName=TotalPower}
  )
  n13 = VertexColor()
  n14 = ScalarParameter() {parameterName=VertexColorInfluence}
  n15 = TextureCoordinate()
  n16 = Abs(None=Add(A=ComponentMask(None=@n15) {r=true}) {constB=-0.5})
  n17 = Constant() {r=0.5}
  n18 = Constant() {r=0.08}
  n19 = Power(
   Base=OneMinus(None=MaterialFunctionCall!(Texture=@n16, Erosion=@n17, Softness=@n18) {materialFunction=/Game/Effects/Niagara/MasterMaterials/Functions/MF_Erosion_Smoothstep_01}.Result)
   Exp=ScalarParameter() {defaultValue=7, parameterName=SideGradientPower}
  ) {constExponent=20}
  n20 = ScalarParameter() {defaultValue=7, parameterName=SideGradientStrength}
  n21 = Saturate(None=Multiply(A=LinearInterpolate(A=@n12, B=@n13.R, Alpha=@n14), B=Multiply(A=@n19, B=@n20)))
  n22 = VertexColor()
  Multiply(A=@n5.A, B=Multiply(A=ScalarParameter() {defaultValue=0.1, parameterName=Opacity}, B=Multiply(A=@n21, B=@n22.R))) {constA=1}
 ]
}
