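# Translucent unlit Niagara ribbon and mesh-particle material for the center water wave trail of the Spring Break glider.
# Distorted, panning sine waves, edge masks and two multiplied noise layers blend water and wake colors and carve opacity.
# Dynamic-parameter erosion reveals the trail, with inner mask, view-angle fade and quality-switched depth fade.
# World position offset adds camera offset, noise displacement and a clamped, UV-weighted bend driven by a dynamic position parameter.
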Material /BRCosmetics/Gadgets/Assets/VinderTech_GliderChute/Glider_SpringBreak/FX/M_Glider_SpringBreak_WaterWaveTrail_Center
graph {
 properties {
  bAutomaticallySetUsageInEditor=false
  bUsedWithNiagaraMeshParticles=true
  bUsedWithNiagaraRibbons=true
  blendMode=BLEND_Translucent
  pixelDepthOffsetMode=PDOM_Legacy
  shadingModel=MSM_Unlit
  twoSided=true
  usedShadingModels=Unlit
 }
 nodes [
  n1 = ScalarParameter() {parameterName=Camera Offset, sliderMax=30}
  n2 = Multiply(
   A=TextureCoordinate()
   B=AppendVector(
    A=ScalarParameter() {defaultValue=1, group=WPO, parameterName=WPO Scale X, sliderMax=5}
    B=ScalarParameter() {defaultValue=1, group=WPO, parameterName=WPO Scale Y, sliderMax=5}
   )
  )
  n3 = AppendVector(
   A=ScalarParameter() {defaultValue=-1, group=WPO, parameterName=WPO Speed X, sliderMax=5, sliderMin=-5}
   B=ScalarParameter() {group=WPO, parameterName=WPO Speed Y, sliderMax=5, sliderMin=-5}
  )
  n4 = Subtract(A=TextureSampleParameter2D(UVs=Panner(Coordinate=@n2, Speed=@n3)) {group=WPO, parameterName=WPO Noise Texture, texture=/Game/Effects/Fort_Effects/Textures/Tile/T_GaussianNoise_01_M}.R) {constB=0.5}
  n5 = Multiply(A=Constant3Vector() {constant=<rgba 0 0 1 0>}, B=ScalarParameter() {group=WPO, parameterName=WPO Z, sliderMax=100})
  n6 = VectorParameter() {parameterName=WPO Axis Mask, defaultValue=<rgba 1 1 1 1>}
  n7 = DynamicParameter() {parameterIndex=1, paramNames [Position X, Position Y, Position Z, Param4]}
  n8 = ScalarParameter() {defaultValue=20, group=WPO, parameterName=WPO Max Offset, sliderMax=50}
  n9 = Clamp(
   None=Subtract(
    A=MaterialFunctionCall(X=@n7.Position X, Y=@n7.Position X, Z=@n7.Position X) {materialFunction=&MakeFloat3}.Result
    B=ParticlePositionWS()
   )
   Min=Multiply(A=@n8) {constB=-1}
   Max=@n8
  )
  n10 = ComponentMask(None=TextureCoordinate()) {r=true}
  n11 = ScalarParameter() {defaultValue=1, group=WPO, parameterName=WPO Bend, sliderMax=10}
  Add(
   A=Add(
    A=MaterialFunctionCall(Offset Amount=@n1) {materialFunction=/Engine/Functions/Engine_MaterialFunctions02/WorldPositionOffset/CameraOffset}.Camera Offset
    B=Multiply(A=@n4, B=@n5)
   )
   B=LinearInterpolate(B=Multiply(A=@n6.RGB, B=@n9), Alpha=Power(Base=@n10, Exp=@n11))
  )
  n12 = VectorParameter() {parameterName=Water Color, defaultValue=<rgba 0.1558 0.3171 0.5625 1>}
  n13 = VectorParameter() {parameterName=Wake Color, defaultValue=<rgba 1 1 1 1>}
  n14 = Multiply(A=NamedRerouteUsage(), B=NamedRerouteUsage())
  n15 = Constant()
  n16 = ScalarParameter() {defaultValue=-0.5, parameterName=Circle Wave Speed, sliderMax=5, sliderMin=-5}
  n17 = NamedRerouteUsage()
  n18 = AppendVector(
   A=ScalarParameter() {defaultValue=1, group=Distortion, parameterName=Distortion Scale X, sliderMax=5}
   B=ScalarParameter() {defaultValue=1, group=Distortion, parameterName=Distortion Scale Y, sliderMax=5}
  )
  n19 = ScalarParameter() {defaultValue=-0.1, group=Distortion, parameterName=Distortion Speed X, sliderMax=5, sliderMin=-5}
  n20 = ScalarParameter() {group=Distortion, parameterName=Distortion Speed Y, sliderMax=5, sliderMin=-5}
  n21 = TextureSampleParameter2D(UVs=Panner(Coordinate=Multiply(A=@n17, B=@n18), Speed=AppendVector(A=@n19, B=@n20))) {group=Distortion, parameterName=Distortion Texture, texture=&T_GaussianNoise_01_M}
  n22 = AppendVector(A=@n21.R, B=@n21.R)
  n23 = ScalarParameter() {defaultValue=0.5, group=Distortion, parameterName=Distortion Bias, sliderMax=1, sliderMin=-1}
  n24 = Add(
   A=Panner(Coordinate=NamedRerouteUsage(), Speed=AppendVector(A=@n15, B=@n16)) {speedY=-0.5}
   B=Multiply(A=Subtract(A=@n22, B=@n23) {constB=0.5}, B=ScalarParameter() {defaultValue=0.2, group=Distortion, parameterName=UV Distortion, sliderMax=2})
  )
  n25 = Constant()
  n26 = Sine(None=Multiply(A=Distance(A=@n24, B=@n25), B=ScalarParameter() {defaultValue=3, parameterName=Circle Wave Frequency, sliderMax=10}) {constB=10})
  n27 = OneMinus(None=ComponentMask(None=NamedRerouteUsage()) {g=true})
  n28 = Multiply(A=Add(A=ConstantBiasScale(None=@n26) {bias=2, scale=0.25}, B=@n27), B=@n27)
  n29 = TextureCoordinate()
  n30 = Sine(None=ComponentMask(None=@n29) {r=true}) {period=2}
  n31 = Sine(None=ComponentMask(None=@n29) {g=true}) {period=2}
  n32 = Multiply(A=Saturate(None=Multiply(A=@n30) {constB=2}), B=Saturate(None=Multiply(A=@n31) {constB=2}))
  n33 = Multiply(
   A=Power(Base=Multiply(A=@n28, B=@n32), Exp=ScalarParameter() {defaultValue=2, parameterName=Gradient Power, sliderMax=5})
   B=ScalarParameter() {defaultValue=2, parameterName=Gradient Multiply, sliderMax=5}
  )
  n34 = Power(Base=Multiply(A=@n14, B=@n33), Exp=ScalarParameter() {defaultValue=1, parameterName=Color Power, sliderMax=2}) {constExponent=0.5}
  n35 = ParticleColor()
  MaterialFunctionCall(Emissive=Multiply(A=LinearInterpolate(A=@n12.RGB, B=@n13.RGB, Alpha=@n34), B=@n35.RGB)) {materialFunction=&MF_FX_EmissiveAdjustments}
  n36 = TextureCoordinate()
  n37 = AppendVector(
   A=ScalarParameter() {defaultValue=1, group=Noise, parameterName=Scale X, sliderMax=5}
   B=ScalarParameter() {defaultValue=1, group=Noise, parameterName=Scale Y, sliderMax=5}
  )
  n38 = MaterialFunctionCall(XY=MaterialFunctionCall(UVs=@n36, Texture Scale=@n37) {materialFunction=&ScaleUVsByCenter}.UVs) {materialFunction=&Swizzle}
  MaterialFunctionCall(Vector or UVs=Subtract(A=@n38.YX, B=Constant2Vector() {r=0.5})) {materialFunction=&VectorToRadialValue}
  n39 = Multiply(A=@n14, B=ScalarParameter() {defaultValue=1, parameterName=Noise Multiply, sliderMax=5})
  n40 = Multiply(A=@n35.A, B=Saturate(None=Subtract(A=@n33, B=@n39)))
  n41 = DepthFade(Opacity=@n40, FadeDistance=ScalarParameter() {defaultValue=10, parameterName=FadeDistance, sliderMax=100})
  n42 = Multiply(
   A=QualitySwitch(Default=@n41, Low=@n40, High=@n41, Medium=@n40, Epic=@n41)
   B=MaterialFunctionCall("FadeSensitivity (optional)"=ScalarParameter() {parameterName="FadeSensitivity (optional) (S)"}) {materialFunction=&FadeBasedOnViewAngle}.Result
  )
  n43 = NamedRerouteUsage()
  Multiply(A=Multiply(A=Multiply(A=@n42, B=@n43), B=NamedRerouteUsage()), B=ScalarParameter() {defaultValue=1, parameterName=GlobalOpacity})
  Multiply(A=@n36, B=@n37)
  n44 = Panner(
   Coordinate=NamedRerouteUsage()
   Speed=AppendVector(
    A=ScalarParameter() {group=Noise, parameterName=Noise Speed X, sliderMax=5, sliderMin=-5}
    B=ScalarParameter() {defaultValue=-0.5, group=Noise, parameterName=Noise Speed Y, sliderMax=5, sliderMin=-5}
   )
  )
  n45 = AppendVector(
   A=ScalarParameter() {defaultValue=1, group=Noise, parameterName=Noise Scale X, sliderMax=5}
   B=ScalarParameter() {defaultValue=1, group=Noise, parameterName=Noise Scale Y, sliderMax=5}
  )
  NamedRerouteDeclaration(None=TextureSampleParameter2D(UVs=Multiply(A=@n44, B=@n45)) {group=Noise, parameterName=Noise Texture, texture=&T_GaussianNoise_01_M}.R) {name=Noise, nodeColor=<rgba 0.7813 0 1 1>}
  n46 = Panner(
   Coordinate=NamedRerouteUsage()
   Speed=AppendVector(
    A=ScalarParameter() {group=Noise, parameterName=Noise Speed X 2, sliderMax=5, sliderMin=-5}
    B=ScalarParameter() {defaultValue=-0.2, group=Noise, parameterName=Noise Speed Y 2, sliderMax=5, sliderMin=-5}
   )
  )
  n47 = AppendVector(
   A=ScalarParameter() {defaultValue=1, group=Noise, parameterName=Noise Scale X 2, sliderMax=5}
   B=ScalarParameter() {defaultValue=1, group=Noise, parameterName=Noise Scale Y 2, sliderMax=5}
  )
  NamedRerouteDeclaration(None=TextureSampleParameter2D(UVs=Multiply(A=@n46, B=@n47)) {group=Noise, parameterName=Noise Texture 2, texture=&T_GaussianNoise_01_M}.R) {name=Noise_2, nodeColor=<rgba 0.7812 0 1 1>}
  NamedRerouteDeclaration(None=@n38.YX) {name=Radial UV, nodeColor=<rgba 0.1953 1 0 1>}
  n48 = Distance(
   A=MaterialFunctionCall(Float2=TextureCoordinate()) {materialFunction=&BreakOutFloat2Components}.G
   B=Constant() {r=0.5}
  )
  NamedRerouteDeclaration(None=LinearInterpolate(A=ScalarParameter() {defaultValue=0.2, parameterName=InnerOpacityMultiplier}, Alpha=Multiply(A=@n48) {constB=2})) {name=InnerMask, nodeColor=<rgba 1 0.7969 0 1>}
  n49 = MaterialFunctionCall(Float2=TextureCoordinate()) {materialFunction=&BreakOutFloat2Components}
  NamedRerouteDeclaration(None=MaterialFunctionCall!(
    Texture=OneMinus(None=@n49.R)
    Erosion=DynamicParameter() {defaultValue=<rgba -0.3 1 1 1>, paramNames [Speed X, Reveal, Param3, Param4]}.Reveal
    Softness=Constant() {r=0.4}
   ) {materialFunction=/Game/Effects/Niagara/MasterMaterials/Functions/MF_Erosion_Smoothstep_01}.Result) {name=RevealMask, nodeColor=<rgba 0 1 0.9531 1>}
  OneMinus(None=@n10)
  TextureCoordinate()
 ]
}
