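# Translucent unlit Niagara ribbon/mesh material for the water wave trail of the Spring Break glider.
# World position offset combines SplineThicken, panning noise, a sine wave and a clamped, UV-weighted pull toward a dynamic-parameter position.
# Panning trail, noise and caustic textures with UV distortion form a base mask that drives a two-color gradient.
# Particle alpha, length fade, erosion reveal and soft sphere mask shape opacity, with depth fade on High/Epic only.
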Material /BRCosmetics/Gadgets/Assets/VinderTech_GliderChute/Glider_SpringBreak/FX/M_Glider_SpringBreak_WaterWaveTrail_SplineThicken
graph {
 properties {
  bAutomaticallySetUsageInEditor=false
  bUsedWithNiagaraMeshParticles=true
  bUsedWithNiagaraRibbons=true
  blendMode=BLEND_Translucent
  pixelDepthOffsetMode=PDOM_Legacy
  shadingModel=MSM_Unlit
  twoSided=true
  usedShadingModels=Unlit
 }
 nodes [
  n1 = TextureCoordinate()
  n2 = ScalarParameter() {defaultValue=1, group=WPO, parameterName=WPO Scale X, sliderMax=5}
  n3 = ScalarParameter() {defaultValue=1, group=WPO, parameterName=WPO Scale Y, sliderMax=5}
  n4 = Panner(
   Coordinate=Multiply(
    A=MaterialFunctionCall(XY=@n1) {materialFunction=&Swizzle}.YX
    B=AppendVector(A=@n2, B=@n3)
   )
   Speed=AppendVector(
    A=ScalarParameter() {defaultValue=1, group=WPO, parameterName=WPO Speed X, sliderMax=5, sliderMin=-5}
    B=ScalarParameter() {group=WPO, parameterName=WPO Speed Y, sliderMax=5, sliderMin=-5}
   )
  )
  n5 = Multiply(
   A=Subtract(A=TextureSampleParameter2D(UVs=@n4) {group=WPO, parameterName=WPO Noise Texture, texture=/Game/Effects/Fort_Effects/Textures/Tile/T_GaussianNoise_01_M}.R) {constB=0.5}
   B=Multiply(A=Constant3Vector() {constant=<rgba 0 0 1 0>}, B=ScalarParameter() {group=WPO, parameterName=WPO Z, sliderMax=100})
  )
  n6 = ScalarParameter() {defaultValue=80, parameterName=Spline Base}
  n7 = ScalarParameter() {defaultValue=1, parameterName=SplineWidthMultiplier}
  n8 = MaterialFunctionCall(WidthBase=Multiply(A=@n6, B=@n7), WidthTip=Multiply(A=ScalarParameter() {defaultValue=50, parameterName=Spline Tip}, B=@n7)) {materialFunction=&SplineThicken}
  n9 = DynamicParameter() {parameterIndex=1, paramNames [Position X, Position Y, Position Z, Param4]}
  n10 = MaterialFunctionCall(X=@n9.Position X, Y=@n9.Position X, Z=@n9.Position X) {materialFunction=&MakeFloat3}
  n11 = ParticlePositionWS()
  n12 = ScalarParameter() {defaultValue=200, group=WPO, parameterName=WPO Max Offset, sliderMax=50}
  n13 = Multiply(
   A=VectorParameter() {parameterName=WPO Axis Mask, defaultValue=<rgba 1 1 1 1>}.RGB
   B=Clamp(None=Subtract(A=@n10.Result, B=@n11), Min=Multiply(A=@n12) {constB=-1}, Max=@n12)
  )
  n14 = TextureCoordinate()
  n15 = Power(
   Base=OneMinus(None=ComponentMask(None=@n14) {g=true})
   Exp=ScalarParameter() {defaultValue=4, group=WPO, parameterName=WPO Bend, sliderMax=10}
  )
  Add(A=Add(A=Add(A=@n5, B=@n8.WorldPositionOffset), B=LinearInterpolate(B=@n13, Alpha=@n15)), B=NamedRerouteUsage())
  n16 = VectorParameter() {parameterName=ColorA, defaultValue=<rgba 0.6 0.7667 1 1>}
  n17 = VectorParameter() {parameterName=ColorB, defaultValue=<rgba 3.6 3.6667 4 1>}
  n18 = Power(Base=Saturate(None=NamedRerouteUsage()), Exp=ScalarParameter() {defaultValue=1, parameterName=ColorPower})
  MaterialFunctionCall(Emissive=Multiply(A=LinearInterpolate(A=@n16.RGB, B=@n17.RGB, Alpha=@n18), B=ParticleColor().RGB)) {materialFunction=&MF_FX_EmissiveAdjustments}
  Multiply(B=@n6)
  n19 = Multiply(A=Multiply(A=ParticleColor().A, B=NamedRerouteUsage()), B=NamedRerouteUsage())
  n20 = DepthFade(Opacity=@n19, FadeDistance=ScalarParameter() {defaultValue=10, parameterName=FadeDistance, sliderMax=100})
  n21 = QualitySwitch(Default=@n19, Low=@n19, High=@n20, Medium=@n19, Epic=@n20)
  n22 = NamedRerouteUsage()
  Multiply(A=Multiply(A=Multiply(A=@n21, B=@n22), B=NamedRerouteUsage()), B=ScalarParameter() {defaultValue=1, parameterName=OpacityMultiplier})
  n23 = TextureCoordinate()
  n24 = Add(
   A=Multiply(
    A=MaterialFunctionCall(Float2=@n23) {materialFunction=&BreakOutFloat2Components}.G
    B=ScalarParameter() {defaultValue=1.333, group=Wave, parameterName=WaveFreq}
   )
   B=Multiply(A=Time(), B=ScalarParameter() {defaultValue=1, group=Wave, parameterName=WaveSpeedMult})
  )
  n25 = Transform(None=Multiply(A=Sine(None=@n24), B=Constant3Vector() {constant=<rgba 1 0 0 0>}))
  n26 = TextureCoordinate()
  n27 = Multiply(
   A=MaterialFunctionCall(Float2=MaterialFunctionCall(XY=@n26) {materialFunction=&Swizzle}.YX) {materialFunction=&BreakOutFloat2Components}.R
   B=ScalarParameter() {defaultValue=20, group=Wave, parameterName=WaveAmount}
  )
  n28 = MaterialFunctionCall(Float2=TextureCoordinate()) {materialFunction=&BreakOutFloat2Components}
  NamedRerouteDeclaration(None=Multiply(A=Multiply(A=@n25, B=@n27), B=OneMinus(None=@n28.G))) {name=Wave, nodeColor=<rgba 1 0 0.0469 1>}
  n29 = MaterialFunctionCall(XY=TextureCoordinate()) {materialFunction=&Swizzle}
  n30 = MaterialFunctionCall(X=ScalarParameter() {defaultValue=1.5, parameterName=Tex2_UScale}, Y=Constant() {r=1}) {materialFunction=&MakeFloat2}
  n31 = Constant() {r=1}
  n32 = TextureCoordinate()
  n33 = LinearInterpolate(
   A=Constant() {r=0.3}
   B=Constant() {r=1}
   Alpha=Power(
    Base=MaterialFunctionCall(Float2=@n32) {materialFunction=&BreakOutFloat2Components}.G
    Exp=Constant() {r=2}
   )
  )
  n34 = Add(
   A=MaterialFunctionCall(
    UVs=Multiply(A=@n29.YX, B=@n30.Result)
    Texture Scale=MaterialFunctionCall(X=@n31, Y=@n33) {materialFunction=&MakeFloat2}.Result
   ) {materialFunction=&ScaleUVsByCenter}.UVs
   B=MaterialFunctionCall(X=DynamicParameter() {defaultValue=<rgba 0 1 1 0>, paramNames [RandomUVOffset, Reveal, U_Scale, U offset]}.RandomUVOffset, Y=Constant()) {materialFunction=&MakeFloat2}.Result
  )
  n35 = NamedRerouteUsage()
  n36 = ScalarParameter() {defaultValue=1.3, parameterName=Tex2_USpeed}
  n37 = Constant()
  n38 = MaterialFunctionCall(Float2=Panner(
    Coordinate=Add(A=@n34, B=@n35)
    Speed=MaterialFunctionCall(X=@n36, Y=@n37) {materialFunction=&MakeFloat2}.Result
   ) {speedX=-0.5}) {materialFunction=&BreakOutFloat2Components}
  n39 = Clamp(None=@n38.G)
  n40 = Power(
   Base=TextureSample(UVs=MaterialFunctionCall(X=@n38.R, Y=@n39) {materialFunction=&MakeFloat2}.Result) {texture=/Game/Gadgets/Assets/VinderTech_GliderChute/Glider_DarkStance/FX/T_Glider_DarkStance_Trail}.G
   Exp=Constant() {r=2}
  )
  n41 = ScalarParameter() {defaultValue=2, parameterName=InnerMultiplier}
  n42 = TextureCoordinate()
  n43 = DynamicParameter() {defaultValue=<rgba 0 1 1 0>, paramNames [RandomUVOffset, Reveal, U_Scale, U offset]}
  n44 = Constant()
  n45 = Multiply(
   A=AppendVector(A=ScalarParameter() {defaultValue=2, parameterName=Tex1_UScale}, B=Constant() {r=1})
   B=Add(
    A=MaterialFunctionCall(XY=@n42) {materialFunction=&Swizzle}.YX
    B=MaterialFunctionCall(X=@n43.RandomUVOffset, Y=@n44) {materialFunction=&MakeFloat2}.Result
   )
  )
  n46 = NamedRerouteUsage()
  n47 = ScalarParameter() {defaultValue=1.5, parameterName=Tex1_USpeed}
  n48 = Constant()
  n49 = TextureSample(UVs=Panner(
    Coordinate=Add(A=@n45, B=@n46)
    Speed=MaterialFunctionCall(X=@n47, Y=@n48) {materialFunction=&MakeFloat2}.Result
   ) {speedX=-0.3}) {
   samplerType=SAMPLERTYPE_LinearColor
   texture=/BRCosmetics/Athena/Effects/Particles/Skydiving/NS_Contrail_StinkyCloud/T_Contrail_StinkyCloud_Ribbon
  }
  n50 = Constant() {r=1}
  n51 = MaterialFunctionCall(XY=TextureCoordinate()) {materialFunction=&Swizzle}
  n52 = MaterialFunctionCall(X=Constant() {r=5}, Y=Constant() {r=1}) {materialFunction=&MakeFloat2}
  n53 = Constant() {r=0.5}
  n54 = Constant()
  n55 = TextureSample(UVs=Panner(
    Coordinate=Multiply(A=@n51.YX, B=@n52.Result)
    Speed=MaterialFunctionCall(X=@n53, Y=@n54) {materialFunction=&MakeFloat2}.Result
   ) {speedX=0.5}) {
   samplerType=SAMPLERTYPE_LinearColor
   texture=/Engine/Functions/Engine_MaterialFunctions02/ExampleContent/Textures/OffsetNoiseDistanceFields_ForNormals
  }
  n56 = OneMinus(None=@n55.R)
  n57 = Multiply(A=@n49.R, B=Power(Base=LinearInterpolate(A=@n50, Alpha=@n56) {constA=0.4}, Exp=Constant() {r=2}) {constExponent=5})
  n58 = ScalarParameter() {defaultValue=0.2, parameterName=OuterMultiplier}
  n59 = Saturate(None=Add(A=Multiply(A=@n40, B=@n41), B=Multiply(A=@n57, B=@n58)))
  n60 = ComponentMask(None=TextureCoordinate()) {r=true}
  n61 = Power(Base=LinearInterpolate(Alpha=@n55.R) {constA=0.5}, Exp=LinearInterpolate(Alpha=Power(Base=@n60) {constExponent=1.5})) {constExponent=1}
  n62 = ComponentMask(None=TextureCoordinate()) {r=true}
  n63 = Multiply(A=Multiply(A=Multiply(A=@n59, B=@n61), B=LinearInterpolate(Alpha=@n62) {constA=0.8}), B=Constant() {r=0.8})
  n64 = AppendVector(A=Constant() {r=2}, B=Constant() {r=2})
  n65 = MaterialFunctionCall(XY=TextureCoordinate() {uTiling=1.1, vTiling=1.7}) {materialFunction=&Swizzle}
  n66 = Constant() {r=1.5}
  n67 = Constant()
  n68 = Add(
   A=Panner(
    Coordinate=Multiply(A=@n64, B=@n65.YX)
    Speed=MaterialFunctionCall(X=@n66, Y=@n67) {materialFunction=&MakeFloat2}.Result
   ) {speedX=2}
   B=NamedRerouteUsage()
  )
  n69 = Multiply(A=Power(Base=TextureSample(UVs=@n68) {samplerType=SAMPLERTYPE_LinearGrayscale, texture=/Game/Global/Textures/Default/Caustics/T_Nosie_Caustics_Grayscale}.R) {constExponent=1}) {constB=1.5}
  n70 = TextureCoordinate()
  n71 = Power(Base=OneMinus(None=ComponentMask(None=@n70) {r=true})) {constExponent=1}
  n72 = Multiply(
   A=LinearInterpolate(A=ScalarParameter() {defaultValue=0.5, parameterName=AdditionalTexBase}, Alpha=Multiply(A=@n69, B=@n71)) {constA=0.5}
   B=@n59
  ) {constA=1}
  NamedRerouteDeclaration(None=Saturate(None=Add(A=@n63, B=@n72))) {name=Base, nodeColor=<rgba 0 0.2266 1 1>}
  n73 = Multiply(
   A=MaterialFunctionCall(Float2=TextureCoordinate()) {materialFunction=&BreakOutFloat2Components}.G
   B=Constant() {r=1.5}
  )
  n74 = Constant() {r=2}
  NamedRerouteDeclaration(None=Clamp(None=Power(Base=@n73, Exp=@n74) {constExponent=4})) {name=OpacityAlongUV, nodeColor=<rgba 0.5703 0 1 1>}
  n75 = TextureCoordinate()
  NamedRerouteDeclaration(None=MaterialFunctionCall!(
    Texture=MaterialFunctionCall(Float2=@n75) {materialFunction=&BreakOutFloat2Components}.G
    Erosion=DynamicParameter() {defaultValue=<rgba 0 1 1 0>, paramNames [RandomUVOffset, Reveal, U_Scale, U offset]}.Reveal
    Softness=Constant() {r=0.4}
   ) {materialFunction=/Game/Effects/Niagara/MasterMaterials/Functions/MF_Erosion_Smoothstep_01}.Result) {name=RevealMask, nodeColor=<rgba 0 1 0.9531 1>}
  n76 = TextureCoordinate()
  n77 = TextureCoordinate()
  n78 = ScalarParameter() {defaultValue=1, group=Distortion, parameterName=Distortion Scale X, sliderMax=5}
  n79 = ScalarParameter() {defaultValue=1, group=Distortion, parameterName=Distortion Scale Y, sliderMax=5}
  n80 = Panner(
   Coordinate=Multiply(
    A=MaterialFunctionCall(XY=@n77) {materialFunction=&Swizzle}.YX
    B=AppendVector(A=@n78, B=@n79)
   )
   Speed=AppendVector(A=Constant(), B=ScalarParameter() {group=Distortion, parameterName=Distortion Speed Y, sliderMax=5, sliderMin=-5})
  )
  n81 = AppendVector(A=DynamicParameter() {defaultValue=<rgba 0 1 1 0>, paramNames [RandomUVOffset, Reveal, U_Scale, U offset]}.RandomUVOffset, B=Constant())
  n82 = Subtract(A=TextureSampleParameter2D(UVs=Add(A=@n80, B=@n81)) {group=Noise, parameterName=Noise Texture, texture=&T_GaussianNoise_01_M}.R) {constB=0.5}
  n83 = ScalarParameter() {defaultValue=0.1, group=Distortion, parameterName=UV Distortion, sliderMax=2}
  n84 = MaterialFunctionCall(
   UVs=Add(
    A=MaterialFunctionCall(XY=@n76) {materialFunction=&Swizzle}.YX
    B=Multiply(A=@n82, B=@n83)
   )
   Texture Scale=AppendVector(
    A=ScalarParameter() {defaultValue=1, parameterName=Scale X, sliderMax=5}
    B=ScalarParameter() {defaultValue=1, parameterName=Scale Y, sliderMax=5}
   )
  ) {materialFunction=&ScaleUVsByCenter}
  n85 = ComponentMask(None=@n84.UVs) {r=true}
  n86 = Constant() {r=0.58}
  n87 = Constant() {r=0.38}
  n88 = ComponentMask(None=@n84.UVs) {g=true}
  n89 = Constant() {r=0.5}
  NamedRerouteDeclaration(None=Multiply(
    A=SphereMask(A=@n85, B=@n86, Radius=@n87) {attenuationRadius=0.4, hardnessPercent=0}
    B=SphereMask(A=@n88, B=@n89) {attenuationRadius=0.5, hardnessPercent=0}
   ) {constA=1}) {name=SphereMask, nodeColor=<rgba 0 1 0.9766 1>}
  n90 = MaterialFunctionCall(XY=TextureCoordinate()) {materialFunction=&Swizzle}
  n91 = Constant() {r=2}
  n92 = TextureSample(UVs=Panner(Coordinate=Multiply(A=@n90.YX, B=@n91)) {speedX=0.2, speedY=0.1}) {texture=/BRCosmetics/Athena/Effects/Particles/Skydiving/NS_Contrail_PinkSpike/Textures/T_PinkSpike_Voronoi2D_Height}
  NamedRerouteDeclaration(None=Multiply(A=ComponentMask(None=@n92.RGB) {g=true, r=true}, B=ScalarParameter() {parameterName=DistortionStrength}) {constB=0.12}) {name=UVDistort, nodeColor=<rgba 0 0.3906 1 1>}
 ]
}
