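# Material for the Spring Break glider that adds animated wind ripples between the character parent PreFX and PostFX functions.
# A wind normal map pans along SpecialUV-derived, mask-selected coordinates, with VelocityAlpha driving speed and negative-flatness exaggeration.
# The ripples blend into the base normal and feed a Fresnel-based fuzzy emissive tinted by time of day.
# A WindArea reroute from SpecialUV's blue channel provides the effect mask.
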Material /BRCosmetics/Gadgets/Assets/VinderTech_GliderChute/Glider_SpringBreak/Materials/M_Glider_SpringBreak_Wind
graph {
 properties {
  bUseMaterialAttributes=true
  bUsedWithMeshDeformer=true
  bUsedWithNiagaraMeshParticles=true
  bUsedWithSkeletalMesh=true
  pixelDepthOffsetMode=PDOM_Legacy
  shadingModel=MSM_FromMaterialExpression
  usedShadingModels=Default Lit | Subsurface | Clear Coat | Subsurface Profile | Cloth
 }
 nodes [
  n1 = MaterialFunctionCall() {materialFunction=&MF_CharacterParent_2023_PreFX}
  n2 = GetMaterialAttributes(None=@n1.Result) {attributeGetTypes [#1, #2]}
  n3 = ScalarParameter() {defaultValue=0.6, group=Wind, parameterName=Tilling}
  n4 = TextureSampleParameter2D() {
   group=Wind
   parameterName=SpecialUV
   texture=/BRCosmetics/Gadgets/Assets/VinderTech_GliderChute/Glider_SpringBreak/Textures/T_UV_Paint_Glider_SpecialUV
  }
  n5 = TextureSampleParameter2D() {
   parameterName=WindTextureMask
   samplerType=SAMPLERTYPE_Masks
   texture=/BRCosmetics/Gadgets/Assets/VinderTech_GliderChute/Glider_SpringBreak/Textures/T_SpringBreak_AdditionalMask
  }
  n6 = LinearInterpolate(A=AppendVector(A=@n4.R, B=@n4.R), B=AppendVector(A=@n4.G, B=@n4.G), Alpha=@n5.R)
  n7 = ScalarParameter() {group=Wind, parameterName=VelocityAlpha}
  n8 = LinearInterpolate(
   A=ScalarParameter() {defaultValue=0.3, group=Wind, parameterName=SpeedMin}
   B=ScalarParameter() {defaultValue=0.5, group=Wind, parameterName=SpeedMax}
   Alpha=@n7
  )
  n9 = ComponentMask(None=VectorParameter() {parameterName=AnimationDirection, defaultValue=<rgba 0 -1 0 1>}.RGB) {g=true, r=true}
  n10 = TextureSampleParameter2D(UVs=Panner(Coordinate=Multiply(A=@n3, B=@n6), Speed=Multiply(A=@n8, B=@n9)) {speedY=0.3}) {
   group=Wind
   parameterName=Wind_N
   samplerType=SAMPLERTYPE_Normal
   texture=/BRCosmetics/Gadgets/Assets/VinderTech_GliderChute/Glider_SpringBreak/Textures/T_wind_N
  }
  n11 = MaterialFunctionCall(
   Normal=@n10.RGB
   Flatness=LinearInterpolate(
    A=ScalarParameter() {defaultValue=-1, group=Wind, parameterName=FlatnessMax}
    B=ScalarParameter() {defaultValue=-2, group=Wind, parameterName=FlatnessMin}
    Alpha=@n7
   )
  ) {materialFunction=&FlattenNormal}
  n12 = MaterialFunctionCall(BaseNormal=@n2.Normal, AdditionalNormal=@n11.Result) {materialFunction=&BlendAngleCorrectedNormals}
  n13 = NamedRerouteUsage()
  n14 = ScalarParameter() {defaultValue=0.1, group=Wind, parameterName=FuzzyEM}
  n15 = Transform(None=@n11.Result)
  n16 = Fresnel(Normal=@n15) {baseReflectFraction=0, exponent=1}
  n17 = MaterialFunctionCall(
   In=Multiply(A=OneMinus(None=@n16), B=Fresnel(Normal=@n15) {baseReflectFraction=0, exponent=1})
   Contrast=ScalarParameter() {defaultValue=0.1, group=Wind, parameterName=EM_Contrast}
  ) {materialFunction=&CheapContrast}
  n18 = VectorParameter() {group=Wind, parameterName=Morning_Color, defaultValue=<rgba 1 0.7014 0.528 1>}
  n19 = VectorParameter() {group=Wind, parameterName=Day_Color, defaultValue=<rgba 1 1 1 1>}
  n20 = VectorParameter() {group=Wind, parameterName=Evening_Color, defaultValue=<rgba 1 0.5852 0.5153 1>}
  n21 = VectorParameter() {group=Wind, parameterName=Night_Color, defaultValue=<rgba 0.5659 0.5677 1 1>}
  n22 = Multiply(
   A=Multiply(
    A=Multiply(A=@n14, B=@n17.Result)
    B=MaterialFunctionCall(Color_Morning=@n18.RGB, Color_Day=@n19.RGB, Color_Evening=@n20.RGB, Color_Night=@n21.RGB) {materialFunction=&Color_Time_of_Day}.Result
   )
   B=NamedRerouteUsage()
  )
  MaterialFunctionCall(
   Material=SetMaterialAttributes!(
    MaterialAttributes=@n2.MaterialAttributes
    Normal=LinearInterpolate(A=@n2.Normal, B=@n12.Result, Alpha=@n13)
    Emissive Color=Add(A=@n2.Emissive Color, B=@n22)
   ) {attributeSetTypes [#1, #2]}
   M=@n1.M
   Pre Skinned Local Position=@n1.M
   Distance Blend=@n1.M
   Fresnel=@n1.M
  ) {materialFunction=&MF_CharacterParent_2023_PostFX}
  NamedRerouteDeclaration(None=@n4.B) {name=WindArea, nodeColor=<rgba 0 0.5781 1 1>}
 ]
}
